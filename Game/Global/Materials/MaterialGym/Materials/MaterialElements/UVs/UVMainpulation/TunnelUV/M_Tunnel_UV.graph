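# Material Gym example material that demonstrates tunnel UV mapping on a color grid test texture.
# MF_TunnelUv supplies the warped UVs for the texture sample and a mask that blends the result toward black.
# The sample takes its mip derivatives from the unwarped texture coordinates, so the UV distortion does not drive mip selection.
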Material /Game/Global/Materials/MaterialGym/Materials/MaterialElements/UVs/UVMainpulation/TunnelUV/M_Tunnel_UV
graph {
 properties {bAutomaticallySetUsageInEditor=false, bHasPixelAnimation=true, usedShadingModels=Default Lit}
 nodes [
  n1 = MaterialFunctionCall() {materialFunction=&/MF_TunnelUv}
  n2 = TextureCoordinate()
  LinearInterpolate(
   A=TextureSample!(UVs=@n1.TunnelUV, Tex, "DDX(UVs)"=DDX(Value=@n2), "DDY(UVs)"=DDY(Value=@n2), Apply View MipBias) {mipValueMode=TMVM_Derivative, samplerSource=SSM_Wrap_WorldGroupSettings, texture=/UVEditor/Textures/UVEditorColorGrid_Color}.RGB
   B=Constant3Vector()
   Alpha=@n1.TunnelMask
  )
 ]
}
